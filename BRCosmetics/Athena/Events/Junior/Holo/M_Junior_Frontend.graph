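# Translucent unlit hologram material for the Junior event's skeletal mesh, animated by the countdown collection's normalized progress.
# Fake-sun lighting, normal-mapped Fresnel and emissive texture form the color, whose desaturated value also drives opacity with a Fresnel edge term.
# A biased custom depth comparison hides pixels behind that depth, faking translucent sorting.
# Countdown progress shifts a scale-aware world-height fade and interpolates an offset along local Z.
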Material /BRCosmetics/Athena/Events/Junior/Holo/M_Junior_Frontend
graph {
 properties {
  allowTranslucentCustomDepthWrites=true
  bUsedWithMeshDeformer=true
  bUsedWithSkeletalMesh=true
  blendMode=BLEND_Translucent
  pixelDepthOffsetMode=PDOM_Legacy
  shadingModel=MSM_Unlit
  usedShadingModels=Unlit
 }
 nodes [
  n1 = VectorParameter() {parameterName=ColorSun, defaultValue=<rgba 1 1 1 1>}
  n2 = VectorParameter() {parameterName=FakeSunDir, defaultValue=<rgba -40.3092 18.0914 19.3332 1>}
  n3 = TextureSampleParameter2D() {group=Normal, parameterName=Normals, samplerType=SAMPLERTYPE_Normal, texture=&/T_M_MED_Junior_Arms_N}
  n4 = Transform(None=@n3.RGB)
  n5 = Power(
   Base=DotProduct(A=Normalize(VectorInput=@n2.RGB), B=Normalize(VectorInput=@n4))
   Exp=ScalarParameter() {defaultValue=1.8143, parameterName=SinDirectionDotPower}
  )
  n6 = VectorParameter() {parameterName=Color, defaultValue=<rgba 0 0 0 1>}
  n7 = Fresnel(ExponentIn=ScalarParameter() {defaultValue=2.5045, parameterName=Fresnel1}, Normal=@n4) {baseReflectFraction=0}
  n8 = Add(
   A=Add(A=Multiply(A=@n1.RGB, B=@n5), B=Multiply(A=@n6.RGB, B=@n7))
   B=Multiply(
    A=VectorParameter() {parameterName=Emissive, defaultValue=<rgba 0.2031 0.0055 0 1>}.RGB
    B=TextureSampleParameter2D() {parameterName=Emissive, texture=&/T_M_MED_Junior_Arms_E}.RGB
   )
  )
  Add(A=VectorParameter() {parameterName=SkyColor, defaultValue=<rgba 0.0117 0.0505 0.1823 1>}.RGB, B=@n8)
  n9 = ComponentMask(None=SceneTexture() {sceneTextureId=PPI_CustomDepth}.Color) {r=true}
  n10 = Add(A=PixelDepth().R, B=ScalarParameter() {defaultValue=-8.9897, parameterName=BiasSorting_Far})
  Divide(A=Subtract(A=@n9, B=@n10), B=ScalarParameter() {defaultValue=3963.7151, parameterName=Divisor})
  n11 = Add(
   A=Multiply(A=ScalarParameter() {defaultValue=1, parameterName=Opacity1}, B=Desaturation(None=@n8))
   B=ScalarParameter() {defaultValue=0.6223, parameterName=OpacityMin}
  )
  n12 = LinearInterpolate(
   B=ScalarParameter() {defaultValue=-1.8316, parameterName=OpacityEdge}
   Alpha=Fresnel(ExponentIn=ScalarParameter() {defaultValue=5.2394, parameterName=FresnelOpaicity}, Normal=@n4) {baseReflectFraction=0}
  ) {constA=1}
  n13 = Constant()
  n14 = Constant() {r=1}
  n15 = WorldPosition()
  n16 = ComponentMask(None=ObjectPositionWS()) {b=true}
  n17 = ComponentMask(None=CollectionParameter() {collection=/Game/Athena/Events/Generic/CountdownTimer/Materials/MPC_Countdown, parameterName=CountdownNomalizedProgress}) {r=true}
  n18 = LinearInterpolate(A=ScalarParameter() {defaultValue=-34.1778, parameterName=FadeOffsetPos1}, B=Constant(), Alpha=@n17)
  n19 = MaterialFunctionCall() {materialFunction=/Engine/Functions/Engine_MaterialFunctions02/WorldPositionOffset/ObjectScale}
  n20 = Subtract(
   A=Subtract(A=ComponentMask(None=@n15.XYZ) {b=true}, B=Subtract(A=@n16, B=@n18))
   B=Multiply(A=@n19.Scale Z, B=ScalarParameter() {defaultValue=55, parameterName=FadeOffset})
  )
  n21 = Multiply(A=@n19.Scale Z, B=ScalarParameter() {defaultValue=4, parameterName=FadeDistance})
  Multiply(
   A=Multiply(A=Multiply(A=@n11, B=@n12), B=If(A=@n10, B=@n9, "A > B"=@n13, "A == B"=@n13, "A < B"=@n14))
   B=Saturate(None=Divide(A=@n20, B=@n21))
  )
  n22 = ScalarParameter() {defaultValue=-22.6019, parameterName=Pos1}
  n23 = ScalarParameter() {parameterName=Pos2}
  n24 = Constant3Vector() {constant=<rgba 0 0 1 0>}
  Multiply(
   A=Multiply(A=LinearInterpolate(A=@n22, B=@n23, Alpha=@n17), B=@n19.Scale Z)
   B=Normalize(VectorInput=Transform(None=@n24) {transformSourceType=TRANSFORMSOURCE_Local})
  )
  ScalarParameter() {parameterName=RiseControl}
 ]
}
