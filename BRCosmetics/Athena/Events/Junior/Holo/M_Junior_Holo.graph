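# Translucent unlit hologram material for the Junior event mesh, with Fresnel rim color plus textured emissive.
# MPC_Countdown progress drives a local-Z position offset and shifts the world-height opacity fade.
# Opacity also combines Fresnel, a height-based monitor mask, animated sine scanlines, and a CustomDepth versus PixelDepth test for self-sorting.
# When the camera comes within TurnOffDistance, opacity fades out and the vertex offset finally collapses the mesh toward its object position.
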Material /BRCosmetics/Athena/Events/Junior/Holo/M_Junior_Holo
graph {
 properties {
  allowTranslucentCustomDepthWrites=true
  bUsedWithMeshDeformer=true
  bUsedWithSkeletalMesh=true
  blendMode=BLEND_Translucent
  pixelDepthOffsetMode=PDOM_Legacy
  shadingModel=MSM_Unlit
  usedShadingModels=Unlit
 }
 nodes [
  n1 = TextureSampleParameter2D() {group=Normal, parameterName=Normals, samplerType=SAMPLERTYPE_Normal, texture=&/T_M_MED_Junior_Arms_N}
  n2 = Transform(None=@n1.RGB)
  n3 = Fresnel(ExponentIn=ScalarParameter() {defaultValue=4.462, parameterName=Fresnel1}, Normal=@n2) {baseReflectFraction=0}
  Add(
   A=Multiply(A=VectorParameter() {parameterName=Color, defaultValue=<rgba 5 0.5107 0.0724 1>}.RGB, B=@n3)
   B=Multiply(
    A=VectorParameter() {parameterName=Emissive, defaultValue=<rgba 3 0.0815 0 1>}.RGB
    B=TextureSampleParameter2D() {parameterName=Emissive, texture=&/T_M_MED_Junior_Arms_E}.RGB
   )
  )
  ComponentMask(None=MaterialFunctionCall() {materialFunction=&ComponentPivotLocation}.Result) {b=true}
  n4 = ComponentMask(None=SceneTexture() {sceneTextureId=PPI_CustomDepth}.Color) {r=true}
  n5 = Add(A=PixelDepth().R, B=ScalarParameter() {defaultValue=-0.023, parameterName=BiasSorting_Far})
  Divide(A=Subtract(A=@n4, B=@n5), B=ScalarParameter() {defaultValue=3963.7151, parameterName=Divisor})
  n6 = WorldPosition() {worldPositionShaderOffset=WPT_ExcludeAllShaderOffsets}
  n7 = ComponentMask(None=CollectionParameter() {collection=/Game/Athena/Events/Generic/CountdownTimer/Materials/MPC_Countdown, parameterName=CountdownNomalizedProgress}) {r=true}
  n8 = Constant3Vector() {constant=<rgba 0 0 1 0>}
  n9 = Multiply(
   A=LinearInterpolate(A=ScalarParameter() {defaultValue=-144.4972, parameterName=Pos1}, B=ScalarParameter() {parameterName=Pos2}, Alpha=@n7)
   B=Normalize(VectorInput=Transform(None=@n8) {transformSourceType=TRANSFORMSOURCE_Local})
  )
  n10 = Subtract(A=Distance(A=ObjectPositionWS(), B=CameraPositionWS()), B=ScalarParameter() {defaultValue=700, parameterName=TurnOffDistance})
  n11 = ScalarParameter() {defaultValue=100, parameterName=TurnOffFade}
  n12 = ComponentMask(None=Saturate(None=Divide(A=@n10, B=@n11))) {r=true}
  LinearInterpolate(A=Subtract(A=ObjectPositionWS(), B=Add(A=@n9, B=@n6.XYZ)), B=@n9, Alpha=Ceil(None=@n12))
  n13 = ObjectPositionWS()
  n14 = ScalarParameter() {defaultValue=-34.1778, parameterName=FadeOffsetPos1}
  n15 = Constant()
  n16 = Subtract(
   A=ComponentMask(None=WorldPosition().XYZ) {b=true}
   B=Subtract(A=ComponentMask(None=@n13) {b=true}, B=LinearInterpolate(A=@n14, B=@n15, Alpha=@n7))
  )
  n17 = Divide(
   A=Subtract(A=@n16, B=ScalarParameter() {defaultValue=6.0377, parameterName=FadeOffset})
   B=ScalarParameter() {defaultValue=29.9835, parameterName=FadeDistance}
  )
  n18 = ScalarParameter() {defaultValue=0.9582, parameterName=Opacity1}
  n19 = Constant() {r=1}
  n20 = ScalarParameter() {defaultValue=1313.0922, parameterName=DepthFade}
  n21 = Subtract(A=@n16, B=ScalarParameter() {defaultValue=62.3348, parameterName=MonitorFadeOffset})
  n22 = ScalarParameter() {defaultValue=16, parameterName=MonitorFadeDistance}
  n23 = LinearInterpolate(
   A=Constant() {r=1}
   B=LinearInterpolate(A=ScalarParameter() {defaultValue=0.2286, parameterName=MaskMonitor}, Alpha=DepthFade(Opacity=@n19, FadeDistance=@n20))
   Alpha=Saturate(None=Divide(A=@n21, B=@n22))
  )
  n24 = Time()
  n25 = ScalarParameter() {defaultValue=0.4857, parameterName=Speed}
  n26 = Add(A=ComponentMask(None=WorldPosition().XYZ) {b=true}, B=Frac(None=Multiply(A=@n24, B=@n25)))
  n27 = ScalarParameter() {defaultValue=2.2185, parameterName=Wave2}
  n28 = Ceil(None=Sine(None=Divide(A=@n26, B=@n27)))
  n29 = Fresnel(ExponentIn=ScalarParameter() {defaultValue=2.0539, parameterName=Fresnel2}, Normal=@n2) {baseReflectFraction=0}
  n30 = ScalarParameter() {defaultValue=23.8906, parameterName=Wave}
  n31 = Saturate(None=Sine(None=Divide(A=@n26, B=@n30)))
  n32 = Multiply(
   A=LinearInterpolate(A=Multiply(A=@n18, B=@n23), B=ScalarParameter() {defaultValue=-1, parameterName=Opacity2}, Alpha=@n3)
   B=LinearInterpolate(A=Constant() {r=1}, B=Saturate(None=@n28), Alpha=Multiply(A=@n29, B=@n31))
  )
  n33 = Constant()
  n34 = If(A=@n5, B=@n4, "A > B"=@n33, "A == B"=@n33, "A < B"=Constant() {r=1})
  LinearInterpolate(B=Multiply(A=Saturate(None=@n17), B=Multiply(A=@n32, B=@n34)), Alpha=@n12)
  MaterialFunctionCall() {materialFunction=&DebugTimeSine}
  ScalarParameter() {defaultValue=1, parameterName=RiseControl}
 ]
}
